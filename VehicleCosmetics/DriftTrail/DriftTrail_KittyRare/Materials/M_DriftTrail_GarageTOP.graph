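# Translucent unlit material for the Kitty Rare drift trail (GarageTOP variant), used on Niagara ribbons and mesh particles.
# A diagonal-lines texture, with UVs mirrored by the Flip dynamic parameter, supplies the pattern alongside edge-fade masks.
# A panning sine band adds a moving glow to the emissive, which runs through MF_FX_EmissiveAdjustments; particle alpha scales opacity.
# The IsWheelWrap? static switch changes tiling and mask shape.
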Material /VehicleCosmetics/DriftTrail/DriftTrail_KittyRare/Materials/M_DriftTrail_GarageTOP
graph {
 properties {
  bUsedWithNiagaraMeshParticles=true
  bUsedWithNiagaraRibbons=true
  blendMode=BLEND_Translucent
  shadingModel=MSM_Unlit
  usedShadingModels=Unlit
 }
 nodes [
  MaterialExpressionVertexColor_0 = ParticleColor()
  MaterialExpressionConstantClamp_0 = Saturate(None=Multiply(A=@MaterialExpressionVertexColor_0.A, B=NamedRerouteUsage()))
  n1 = ScalarParameter() {defaultValue=-2, parameterName=Speex X Brightness}
  n2 = Constant()
  n3 = ComponentMask(None=Panner(Coordinate=TextureCoordinate() {uTiling=2}, Speed=AppendVector(A=@n1, B=@n2)) {speedX=-2}) {r=true}
  n4 = Saturate(None=Power(Base=Sine(None=@n3) {period=2}) {constExponent=3})
  n5 = Multiply(
   A=NamedRerouteDeclaration(None=@n4) {name=Mask panning, nodeColor=<rgba 0 0.5547 1 1>}
   B=ScalarParameter() {defaultValue=5, parameterName=GlowTrailOutlines}
  ) {constB=5}
  n6 = NamedRerouteUsage()
  n7 = Multiply(A=NamedRerouteUsage(), B=NamedRerouteUsage())
  n8 = NamedRerouteUsage()
  MaterialFunctionCall(Emissive=Add(A=Multiply(A=@n5, B=@n6), B=Add(A=@n7, B=@n8))) {materialFunction=&MF_FX_EmissiveAdjustments}
  n9 = DynamicParameter() {paramNames [OffsetTrail,Flip,flipB,Param4]}
  n10 = If(
   A=Saturate(None=Floor(None=@n9.Flip))
   "A > B"=StaticSwitchParameter(True=TextureCoordinate() {uTiling=2, vTiling=0.2}, False=TextureCoordinate() {vTiling=0.13}) {parameterName=IsWheelWrap?}
   "A < B"=StaticSwitchParameter(True=TextureCoordinate() {uTiling=-2, vTiling=0.2}, False=TextureCoordinate() {vTiling=-0.13}) {parameterName=IsWheelWrap?}
  ) {constB=0.5}
  Multiply(A=@n10, B=AppendVector(A=Constant() {r=1}, B=Constant() {r=0.13}))
  NamedRerouteDeclaration(None=TextureSample(UVs=@n10) {texture=/DealershipSharedLibrary/FX/Textures/Gradient_Masks/T_VCVFX_KittyRareDiagonalLines_A}.RGB) {name=Lines Texture, nodeColor=<rgba 0 1 0.2734 1>}
  n11 = NamedRerouteUsage()
  n12 = Power(Base=@n11) {constExponent=6}
  n13 = Power(Base=OneMinus(None=@n11)) {constExponent=6}
  n14 = Add(A=Saturate(None=Multiply(B=@n12) {constA=2}), B=Saturate(None=Multiply(B=@n13) {constA=2}))
  n15 = Multiply(A=@n14, B=NamedRerouteUsage())
  n16 = TextureCoordinate()
  n17 = Power(Base=Cosine(None=ComponentMask(None=@n16) {g=true})) {constExponent=18}
  n18 = Add(A=Add(A=Multiply(A=@n14, B=@n15), B=Round(None=@n17))) {constB=0.01}
  n19 = NamedRerouteUsage()
  n20 = NamedRerouteUsage()
  n21 = TextureCoordinate()
  n22 = ComponentMask(None=@n21) {r=true}
  n23 = NamedRerouteDeclaration(None=@n22) {name=Masks Coords, nodeColor=<rgba 1 0 0.0035 1>}
  n24 = Saturate(None=Multiply(A=Power(Base=@n23) {constExponent=3}) {constB=500})
  n25 = Power(Base=OneMinus(None=@n23)) {constExponent=5}
  n26 = Multiply(A=Saturate(None=Multiply(A=@n25) {constB=10})) {constB=0.35}
  n27 = Saturate(None=Multiply(A=Multiply(A=@n24, B=@n26)) {constB=4})
  NamedRerouteDeclaration(None=StaticSwitchParameter(True=Multiply(A=@n18, B=@n19), False=Multiply(A=@n20, B=@n27)) {parameterName=IsWheelWrap?}) {name=Mask Opacity, nodeColor=<rgba 0 0.3672 1 1>}
  NamedRerouteDeclaration(None=@MaterialExpressionVertexColor_0.RGB) {name=RGB Particle, nodeColor=<rgba 0.6641 1 0 1>}
  NamedRerouteDeclaration(None=ComponentMask(None=@n21) {g=true}) {name=Masks Coords Y, nodeColor=<rgba 1 0 0.0035 1>}
  Reroute(None=@n4)
  Reroute(None=@n22)
  Transform(None=DynamicParameter() {parameterIndex=1, paramNames [x,y,z,Param4]}.RGB) {transformSourceType=TRANSFORMSOURCE_Local}
 ]
}
