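# VC_MF_SimplePannerFX is a material function in the FX library that outputs scrolling UVs.
# A static switch picks either a supplied UV input or the default texture coordinates multiplied by separate X and Y tiling.
# The chosen UVs feed a Panner whose speed is built from separate X and Y speed inputs, which default to zero.
# Tiling applies only when custom UV is off.
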MaterialFunction /DealershipSharedLibrary/FX/Materials/MaterialFunctions/VC_MF_SimplePannerFX
graph {nodes [
  n1 = TextureCoordinate()
  n2 = FunctionInput(Preview=@n1) {bUsePreviewValueAsDefault=true, inputName=UV, inputType=FunctionInput_Vector2, sortPriority=1}
  n3 = Constant() {r=1}
  n4 = FunctionInput(Preview=Constant() {r=1}) {bUsePreviewValueAsDefault=true, inputName=Tiling Y, inputType=FunctionInput_Scalar, sortPriority=3, previewValue=<0 0 0 0>}
  n5 = Multiply(
   A=@n1
   B=AppendVector(A=FunctionInput(Preview=@n3) {bUsePreviewValueAsDefault=true, inputName=Tiling X, inputType=FunctionInput_Scalar, sortPriority=2}, B=@n4)
  )
  n6 = FunctionInput(Preview=StaticBool()) {bUsePreviewValueAsDefault=true, inputName=Use custom UV, inputType=FunctionInput_StaticBool}
  n7 = Constant()
  n8 = FunctionInput(Preview=@n7) {bUsePreviewValueAsDefault=true, inputName=Panner Speed X, inputType=FunctionInput_Scalar, sortPriority=4}
  n9 = FunctionInput(Preview=@n7) {bUsePreviewValueAsDefault=true, inputName=Panner Speed Y, inputType=FunctionInput_Scalar, sortPriority=5, previewValue=<0 0 0 0>}
  FunctionOutput(None=Panner(Coordinate=StaticSwitch(True=@n2, False=@n5, Value=@n6), Speed=AppendVector(A=@n8, B=@n9)) {speedX=-1}) {outputName=Panner UV}
 ]}
