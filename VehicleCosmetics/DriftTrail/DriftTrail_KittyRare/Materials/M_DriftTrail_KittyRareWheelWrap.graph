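# Translucent unlit Niagara mesh and ribbon material for the KittyRare drift trail wheel wrap.
# Glowing outlines combine a panned, squared kitty mask, an outline mask and particle color.
# A time-driven If chain flips through puma logo frames packed into the channels of three sprite textures.
# Edge falloffs and diagonal stripes shape the alpha masks, while dynamic parameters drive glow, line alpha, logo offset and UV mirroring.
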Material /VehicleCosmetics/DriftTrail/DriftTrail_KittyRare/Materials/M_DriftTrail_KittyRareWheelWrap
graph {
 properties {
  bAutomaticallySetUsageInEditor=false
  bUsedWithNiagaraMeshParticles=true
  bUsedWithNiagaraRibbons=true
  blendMode=BLEND_Translucent
  shadingModel=MSM_Unlit
  usedShadingModels=Unlit
 }
 nodes [
  n1 = Constant2Vector() {r=2}
  n2 = Power(
   Base=TextureSample(UVs=MaterialFunctionCall!(Use custom UV, UV, Tiling X, Tiling Y, Panner Speed X=@n1.R, Panner Speed Y=@n1.R) {materialFunction=/DealershipSharedLibrary/FX/Materials/MaterialFunctions/VC_MF_SimplePannerFX}.Panner UV) {texture=/DealershipSharedLibrary/FX/Textures/Gradient_Masks/T_VCVFX_BoostKittyRareMask_A}.RGB
   Exp=Constant() {r=2}
  )
  n3 = DynamicParameter() {paramNames [Glow Trail Outlines, AlphaLines, OffsetWPO, OffsetLogo]}
  n4 = TextureSample(UVs=MaterialFunctionCall(UVs=TextureCoordinate()) {materialFunction=&ScaleUVsByCenter}.UVs) {samplerType=SAMPLERTYPE_Masks, texture=/DealershipSharedLibrary/FX/Textures/Gradient_Masks/T_VCVFX_WheelWrapOutLines_A}
  n5 = Multiply(A=Multiply(A=@n2, B=@n3.Glow Trail Outlines), B=Multiply(A=Add(A=@n4.RGB), B=ParticleColor().RGB))
  Add(A=@n5)
  Add(A=NamedRerouteUsage())
  Constant() {r=1}
  Constant() {r=1}
  DynamicParameter() {paramNames [Glow Trail Outlines, AlphaLines, OffsetWPO, OffsetLogo]}
  n6 = Power(Base=ComponentMask(None=TextureCoordinate()) {g=true}) {constExponent=1}
  LinearInterpolate(B=@n6) {constAlpha=0}
  MaterialFunctionCall(Emissive=@n5) {materialFunction=&MF_FX_EmissiveAdjustments}
  n7 = TextureCoordinate()
  Multiply(A=Power(Base=ComponentMask(None=@n7) {r=true}) {constExponent=1}, B=@n2)
  n8 = ComponentMask(None=TextureCoordinate()) {g=true}
  Multiply(B=Round(None=Power(Base=@n8) {constExponent=0.25}))
  Multiply(A=NamedRerouteUsage(), B=Round(None=@n6))
  Multiply(A=NamedRerouteUsage(), B=@n6)
  n9 = ParticleColor()
  n10 = NamedRerouteUsage()
  n11 = TextureSample(UVs=NamedRerouteUsage()) {texture=/VehicleCosmetics/DriftTrail/DriftTrail_KittyRare/Textures/T_VCVFX_RGBASpritesPumaCB}
  n12 = TextureSample(UVs=NamedRerouteUsage()) {texture=/VehicleCosmetics/DriftTrail/DriftTrail_KittyRare/Textures/T_VCVFX_RGBASpritesPumaBB}
  n13 = Time()
  n14 = NamedRerouteDeclaration(None=Multiply(A=Frac(None=@n13)) {constB=19}) {name=Frame, nodeColor=<rgba 1 0.8701 0 1>}
  n15 = Floor(None=DynamicParameter() {parameterIndex=1, paramNames [flip,Param2,Param3,Param4]}.flip)
  n16 = Add(
   A=AppendVector(A=DynamicParameter() {paramNames [Glow Trail Outlines, AlphaLines, OffsetWPO, OffsetLogo]}.OffsetLogo, B=Constant() {r=1})
   B=If(A=Saturate(None=@n15), "A > B"=TextureCoordinate(), "A < B"=TextureCoordinate() {vTiling=-1}) {constB=0.5}
  )
  n17 = AppendVector(A=ScalarParameter() {defaultValue=11, parameterName=Logo Tile X Puma}, B=ScalarParameter() {defaultValue=1, parameterName=Logo Tile Y Puma})
  n18 = TextureSample(UVs=NamedRerouteDeclaration(None=Multiply(A=@n16, B=@n17)) {name=Logo UV, nodeColor=<rgba 0 0.9062 1 1>}) {texture=/VehicleCosmetics/DriftTrail/DriftTrail_KittyRare/Textures/T_VCVFX_RGBASpritesPumaAB}
  n19 = If(A=@n14, "A > B"=@n18.G, "A == B"=@n18.G, "A < B"=@n18.G) {constB=4}
  n20 = If(A=@n14, "A > B"=@n18.B, "A == B"=@n19, "A < B"=@n19) {constB=6}
  n21 = If(A=@n10, "A > B"=@n12.R, "A == B"=@n20, "A < B"=@n20) {constB=8}
  n22 = If(A=@n10, "A > B"=@n12.G, "A == B"=@n21, "A < B"=@n21) {constB=10}
  n23 = If(A=@n10, "A > B"=@n12.B, "A == B"=@n22, "A < B"=@n22) {constB=12}
  n24 = If(A=@n10, "A > B"=@n11.R, "A == B"=@n23, "A < B"=@n23) {constB=14}
  n25 = If(A=@n10, "A > B"=@n11.G, "A == B"=@n24, "A < B"=@n24) {constB=16}
  n26 = ComponentMask(None=TextureCoordinate()) {r=true}
  n27 = Constant() {r=100}
  MaterialExpressionConstantClamp_2 = Saturate(None=Multiply(A=Constant() {r=150}, B=Power(Base=@n26, Exp=@n27)))
  n28 = OneMinus(None=@MaterialExpressionConstantClamp_2)
  n29 = ScalarParameter() {defaultValue=4, parameterName=Logo Tile X}
  n30 = ScalarParameter() {defaultValue=0.2, parameterName=Logo Tile Y}
  n31 = TextureSample(UVs=Multiply(A=TextureCoordinate(), B=AppendVector(A=@n29, B=@n30))) {texture=/DealershipSharedLibrary/FX/Textures/Gradient_Masks/T_VCVFX_KittyRareDiagonalLines_A}
  n32 = Multiply(A=@n25, B=Add(A=@n28, B=OneMinus(None=@n31.RGB)))
  n33 = Multiply(A=@n32, B=@n28)
  n34 = Multiply(A=@MaterialExpressionConstantClamp_2, B=@n32)
  n35 = Saturate(None=Multiply(A=@n9.A, B=Add(A=@n33, B=@n34)))
  n36 = ComponentMask(None=TextureCoordinate()) {r=true}
  Multiply(A=@n35, B=OneMinus(None=Round(None=@n36)))
  NamedRerouteDeclaration(None=@n35) {name=Logo Trail OP, nodeColor=<rgba 0 1 0.1328 1>}
  n37 = Add(A=Multiply(A=NamedRerouteUsage(), B=@n9.RGB), B=@n9.RGB)
  NamedRerouteDeclaration(None=@n37) {name=Logo Trail RGB, nodeColor=<rgba 0 1 0.4375 1>}
  NamedRerouteDeclaration(None=@n25) {name=Alpha Logo, nodeColor=<rgba 0.9453 1 0 1>}
  n38 = ComponentMask(None=TextureCoordinate() {uTiling=3}) {g=true}
  OneMinus(None=@n38)
  Power(Base=@n36) {constExponent=2.6}
  n39 = OneMinus(None=ComponentMask(None=TextureCoordinate() {uTiling=3}) {g=true})
  n40 = Constant() {r=6}
  MaterialExpressionConstantClamp_0 = Saturate(None=Multiply(A=Constant() {r=2}, B=Power(Base=@n39, Exp=@n40)))
  n41 = ScalarParameter() {defaultValue=11, parameterName=test pow2}
  n42 = Subtract(A=@MaterialExpressionConstantClamp_0, B=Multiply(A=Power(Base=@n39, Exp=@n41) {constExponent=40}) {constB=0})
  Reroute(None=@n42)
  Reroute(None=@n37)
  Reroute(None=@n25)
  Reroute()
  Reroute(None=@n35)
  n43 = Constant() {r=0.2}
  n44 = TextureCoordinate()
  n45 = ScalarParameter() {parameterName=BGTrail_OffsetX}
  n46 = Constant()
  n47 = ComponentMask(None=Add(
    A=Multiply(A=@n44) {constB=4.5}
    B=MaterialFunctionCall(X=@n45, Y=@n46) {materialFunction=&MakeFloat2}.Result
   )) {g=true}
  n48 = Power(Base=OneMinus(None=@n47), Exp=ScalarParameter() {defaultValue=3, parameterName=BGTrail_PWRight})
  n49 = Multiply(A=@n48, B=Power(Base=@n47, Exp=ScalarParameter() {parameterName=BGTrail_PWLeft}))
  Saturate(None=Multiply(A=Multiply(A=@n43, B=@n49), B=ScalarParameter() {defaultValue=1, parameterName=BGTrail_Brightness}))
  n50 = Multiply(A=@n4.RGB, B=@n3.AlphaLines)
  n51 = Floor(None=DynamicParameter() {parameterIndex=1, paramNames [flip,Param2,Param3,Param4]}.flip)
  n52 = TextureSample(UVs=If(A=Saturate(None=@n51), "A > B"=TextureCoordinate() {uTiling=2, vTiling=0.2}, "A < B"=TextureCoordinate() {uTiling=-2, vTiling=0.2}) {constB=0.5}) {samplerSource=SSM_Wrap_WorldGroupSettings, texture=&T_VCVFX_KittyRareDiagonalLines_A}
  n53 = Multiply(A=Constant() {r=2}, B=Power(Base=@n38, Exp=Constant() {r=6}))
  n54 = Power(Base=@n38, Exp=ScalarParameter() {defaultValue=11, parameterName=test pow2}) {constExponent=40}
  n55 = Add(A=@n42, B=Subtract(A=Saturate(None=@n53), B=Multiply(A=@n54) {constB=0}))
  n56 = Multiply(A=Multiply(A=@n52.RGB, B=@n55), B=@n55)
  Saturate(None=Add(A=Add(A=@n50, B=@n56)) {constB=0.005})
  Sine(None=@n48)
 ]
}
