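# Translucent unlit Niagara ribbon material for the main trail of the KittyRare drift trail vehicle cosmetic.
# It layers a scrolling-brightness outline with diagonal lines, a gradient background band and a tiled puma logo, combined through named reroutes into emissive and opacity.
# The logo animates as a time-driven flipbook, with If chains picking channels from three channel-packed sprite textures.
# Particle color tints and fades layers, while dynamic parameters drive line alpha and a world position offset.
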Material /VehicleCosmetics/DriftTrail/DriftTrail_KittyRare/Materials/M_DriftTrail_MainTrail
graph {
 properties {
  bAutomaticallySetUsageInEditor=false
  bUsedWithNiagaraRibbons=true
  blendMode=BLEND_Translucent
  shadingModel=MSM_Unlit
  usedShadingModels=Unlit
 }
 nodes [
  n1 = NamedRerouteUsage()
  n2 = Power(Base=ComponentMask(None=TextureCoordinate()) {r=true}) {constExponent=0.3}
  n3 = NamedRerouteUsage()
  n4 = Round(None=@n2)
  MaterialFunctionCall(Emissive=LinearInterpolate(A=NamedRerouteUsage(), B=Multiply(A=@n1, B=@n2), Alpha=Multiply(A=@n3, B=@n4))) {materialFunction=&MF_FX_EmissiveAdjustments}
  n5 = NamedRerouteUsage()
  n6 = TextureCoordinate()
  n7 = Power(Base=Cosine(None=ComponentMask(None=@n6) {g=true})) {constExponent=85}
  n8 = Power(Base=ComponentMask(None=TextureCoordinate()) {r=true}) {constExponent=0.7}
  n9 = ComponentMask(None=TextureCoordinate()) {r=true}
  n10 = Multiply(A=Multiply(A=Round(None=@n7), B=OneMinus(None=@n8)), B=Round(None=Power(Base=@n9) {constExponent=0.25}))
  n11 = Constant2Vector() {g=0.2, r=2}
  n12 = Multiply(A=TextureCoordinate(), B=@n11)
  n13 = MaterialFunctionCall(X=ScalarParameter() {parameterName=OutlinesTrail_OffsetX}, Y=Constant()) {materialFunction=&MakeFloat2}
  n14 = TextureSample(UVs=MaterialFunctionCall!(Use custom UV=StaticBool() {value=true}, UV=Add(A=@n12, B=@n13.Result), Tiling X=@n11.R, Tiling Y=@n11.R, Panner Speed X, Panner Speed Y) {materialFunction=/DealershipSharedLibrary/FX/Materials/MaterialFunctions/VC_MF_SimplePannerFX}.Panner UV) {samplerSource=SSM_Wrap_WorldGroupSettings, texture=/DealershipSharedLibrary/FX/Textures/Gradient_Masks/T_VCVFX_KittyRareDiagonalLines_A}
  n15 = OneMinus(None=ComponentMask(None=TextureCoordinate() {uTiling=3}) {r=true})
  n16 = Constant() {r=7.5}
  MaterialExpressionConstantClamp_0 = Saturate(None=Multiply(A=Constant() {r=220}, B=Power(Base=@n15, Exp=@n16)))
  n17 = Subtract(
   A=@MaterialExpressionConstantClamp_0
   B=Power(Base=@n15, Exp=ScalarParameter() {defaultValue=11.6028, parameterName=test pow2}) {constExponent=40}
  )
  n18 = Multiply(A=Multiply(A=@n14.RGB, B=@n17), B=@n17)
  n19 = DynamicParameter() {paramNames [Glow Trail Outlines, AlphaLines, OffsetWPO, OffsetLogo]}
  n20 = Max(A=NamedRerouteUsage(), B=Multiply(A=Add(A=@n10, B=@n18), B=@n19.AlphaLines))
  n21 = ComponentMask(None=TextureCoordinate()) {r=true}
  Multiply(A=Saturate(None=Add(A=@n5, B=@n20)), B=OneMinus(None=Power(Base=@n21) {constExponent=2.6}))
  n22 = ParticleColor()
  n23 = ComponentMask(None=TextureCoordinate()) {r=true}
  n24 = Constant() {r=100}
  MaterialExpressionConstantClamp_2 = Saturate(None=Multiply(A=Constant() {r=150}, B=Power(Base=@n23, Exp=@n24)))
  n25 = OneMinus(None=@MaterialExpressionConstantClamp_2)
  n26 = ScalarParameter() {defaultValue=4, parameterName=Logo Tile X}
  n27 = ScalarParameter() {defaultValue=0.2, parameterName=Logo Tile Y}
  n28 = TextureSample(UVs=Multiply(A=TextureCoordinate(), B=AppendVector(A=@n26, B=@n27))) {texture=&T_VCVFX_KittyRareDiagonalLines_A}
  n29 = Multiply(A=NamedRerouteUsage(), B=Add(A=@n25, B=OneMinus(None=@n28.RGB)))
  n30 = Multiply(A=@n22.A, B=Add(A=Multiply(A=@n29, B=@n25), B=Multiply(A=@MaterialExpressionConstantClamp_2, B=@n29)))
  n31 = TextureCoordinate()
  n32 = Round(None=Power(Base=ComponentMask(None=@n31) {r=true}) {constExponent=2.6})
  NamedRerouteDeclaration(None=Multiply(A=Saturate(None=@n30), B=OneMinus(None=@n32))) {name=Logo Trail OP, nodeColor=<rgba 0 1 0.1328 1>}
  n33 = NamedRerouteUsage()
  NamedRerouteDeclaration(None=Add(A=Multiply(A=@n33, B=@n22.RGB), B=@n22.RGB)) {name=Logo Trail RGB, nodeColor=<rgba 0 1 0.4375 1>}
  n34 = NamedRerouteUsage()
  n35 = TextureSample(UVs=NamedRerouteUsage()) {texture=/VehicleCosmetics/DriftTrail/DriftTrail_KittyRare/Textures/T_VCVFX_RGBASpritesPumaCB}
  n36 = TextureSample(UVs=NamedRerouteUsage()) {texture=/VehicleCosmetics/DriftTrail/DriftTrail_KittyRare/Textures/T_VCVFX_RGBASpritesPumaBB}
  n37 = Time()
  n38 = NamedRerouteDeclaration(None=Multiply(A=Frac(None=@n37)) {constB=19}) {name=Frame, nodeColor=<rgba 1 0.8701 0 1>}
  n39 = Add(A=AppendVector(A=Constant() {r=-0.608}, B=Constant()), B=TextureCoordinate())
  n40 = AppendVector(A=ScalarParameter() {defaultValue=14, parameterName=Logo Tile X Puma}, B=ScalarParameter() {defaultValue=1, parameterName=Logo Tile Y Puma})
  n41 = TextureSample(UVs=NamedRerouteDeclaration(None=Multiply(A=@n39, B=@n40)) {name=Logo UV, nodeColor=<rgba 0 0.9062 1 1>}) {texture=/VehicleCosmetics/DriftTrail/DriftTrail_KittyRare/Textures/T_VCVFX_RGBASpritesPumaAB}
  n42 = If(A=@n38, "A > B"=@n41.G, "A == B"=@n41.G, "A < B"=@n41.G) {constB=4}
  n43 = If(A=@n38, "A > B"=@n41.B, "A == B"=@n42, "A < B"=@n42) {constB=6}
  n44 = If(A=@n34, "A > B"=@n36.R, "A == B"=@n43, "A < B"=@n43) {constB=8}
  n45 = If(A=@n34, "A > B"=@n36.G, "A == B"=@n44, "A < B"=@n44) {constB=10}
  n46 = If(A=@n34, "A > B"=@n36.B, "A == B"=@n45, "A < B"=@n45) {constB=12}
  n47 = If(A=@n34, "A > B"=@n35.R, "A == B"=@n46, "A < B"=@n46) {constB=14}
  NamedRerouteDeclaration(None=If(A=@n34, "A > B"=@n35.G, "A == B"=@n47, "A < B"=@n47) {constB=16}) {name=Alpha Logo, nodeColor=<rgba 0.9453 1 0 1>}
  n48 = ComponentMask(None=TextureCoordinate()) {r=true}
  n49 = ScalarParameter() {defaultValue=-2, parameterName=Speex X Brightness}
  n50 = Constant()
  n51 = ComponentMask(None=Panner(Coordinate=TextureCoordinate() {uTiling=2}, Speed=AppendVector(A=@n49, B=@n50)) {speedX=-2}) {r=true}
  n52 = Saturate(None=Power(Base=Sine(None=@n51) {period=2}) {constExponent=3})
  n53 = Add(A=Multiply(A=Multiply(A=@n48, B=@n52), B=Constant() {r=20}), B=@n10)
  n54 = Multiply(A=Max(A=NamedRerouteUsage(), B=Add(A=@n53, B=@n18)), B=ParticleColor().RGB)
  NamedRerouteDeclaration(None=@n54) {name=Trail Outlines RGB, nodeColor=<rgba 1 0.2344 0 1>}
  n55 = Constant() {r=7}
  n56 = TextureCoordinate()
  n57 = ScalarParameter() {parameterName=BGTrail_OffsetX}
  n58 = Constant()
  n59 = ComponentMask(None=Add(
    A=Multiply(A=@n56) {constB=4.5}
    B=MaterialFunctionCall(X=@n57, Y=@n58) {materialFunction=&MakeFloat2}.Result
   )) {r=true}
  n60 = Multiply(
   A=Power(Base=OneMinus(None=@n59), Exp=ScalarParameter() {defaultValue=3, parameterName=BGTrail_PWRight})
   B=Power(Base=@n59, Exp=ScalarParameter() {defaultValue=1, parameterName=BGTrail_PWLeft})
  )
  n61 = Saturate(None=Multiply(A=Multiply(A=@n55, B=@n60), B=ScalarParameter() {defaultValue=1, parameterName=BGTrail_Brightness}))
  NamedRerouteDeclaration(None=@n61) {name=BGTrail, nodeColor=<rgba 0 0.0859 1 1>}
  Reroute(None=@n54)
  Reroute(None=@n61)
  Transform(None=MaterialFunctionCall(Y=@n19.OffsetWPO) {materialFunction=&MakeFloat3}.Result) {transformSourceType=TRANSFORMSOURCE_Local}
 ]
}
